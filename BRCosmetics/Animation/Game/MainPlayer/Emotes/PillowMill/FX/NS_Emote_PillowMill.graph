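# Niagara system for the PillowMill player emote: a short, cartoon comic-style impact burst.
# Flash_Comic pops a randomly rotated mesh with curve-animated dot and black-outline attributes across three mesh renderers.
# Orbs_Impact scatters flickering, drag-slowed cyan/pink sprites plus a growing ring, using spawn groups to vary color and visibility.
# Speedlines_Toon expands two randomly rotated sprite layers; it and the flash size from System.FXScale.
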NiagaraSystem /BRCosmetics/Animation/Game/MainPlayer/Emotes/PillowMill/FX/NS_Emote_PillowMill
dependencies {
 usedDataInterfaces [&NiagaraDataInterfaceCurve, &NiagaraDataInterfaceArrayColor]
 usedDynamicInputs [&SimulationPosition]
 usedModules [
  &self:SystemSpawnScript.SystemScriptSource.SystemScriptGraph.NiagaraNodeAssignment_0.SetVariables_#1
  &SystemState
 ]
}
emitters {
 Flash_Comic {
  emitterUpdateScript [
   &EmitterState(Inactive Response=::NewEnumerator1, Life Cycle Mode=::NewEnumerator1, Loop Behavior=::NewEnumerator1, Loop Duration=1, UseLoopDelay=false)
   &SpawnBurst_Instantaneous(Spawn Time=0.01, Use Spawn Probability=false, UseLoopCountLimit=false)
  ]
  particleSpawnScript [
   &InitializeParticle(
    Color=<rgba 0.4984 0 0.1259 1>
    Lifetime=0.4
    Mesh Scale=&Multiply_Vector(
     A=&RandomRangeVector(Minimum=<0.88 1.76 1.76>, Maximum=<0.64 1.28 1.28>)
     B=&VectorFromFloat(Value=$System.FXScale)
    )
    Mesh Scale Mode=::NewEnumerator4
    UsePositionOffset=false
    Write Material Random=true
   )
   &DynamicMaterialParameters(
    Index 0 Param 2=45
    Index 0 Param 3=1
    Index 0 Param 4=1
    Module.Index0Param0Write=true
    Module.Index0Param1Write=true
    Module.Index0Param2Write=true
    Module.Index0Param3Write=true
   ) {module=DynamicMaterialParameters001}
   &InitialMeshOrientation(Rotation=&RandomRangeVector(Minimum=<0 0 0>, Maximum=<1 1 1>), Use Rotation Vector=true) {module=InitialMeshOrientation001}
   /SetVariables(Particles.DotPositionOffset:NiagaraPosition=&RandomRangeVector(Minimum=<0 0 10>, Maximum=<0 -5 -5>)) {isSetParameters=true}
  ]
  particleUpdateScript [
   &ParticleState()
   &ScaleMeshSize(Scale Factor=&VectorFromFloat(Value=&FloatFromCurve(FloatCurve=DataInterface() {propertyValues {Curve {keys [
          {arriveTangent=4.1785, interpMode=RCIM_Cubic, leaveTangent=4.1785, tangentMode=RCTM_User, time=0, value=0}
          {arriveTangent=-0.121, interpMode=RCIM_Cubic, leaveTangent=-0.121, tangentMode=RCTM_User, time=0.05, value=1}
          <at 0.2 1.5 RCIM_Cubic>
          {arriveTangent=-0.8333, interpMode=RCIM_Cubic, leaveTangent=-0.8333, time=0.3, value=1}
          {arriveTangent=-0.075, interpMode=RCIM_Cubic, leaveTangent=-1.0526, tangentMode=RCTM_Break, time=0.8, value=1}
          {arriveTangent=0.0224, interpMode=RCIM_Cubic, leaveTangent=0.0224, tangentMode=RCTM_User, time=1, value=0}
         ]}}}))) {module=ScaleMeshSize001}
   &ScaleColor(
    Scale Alpha=&FloatFromCurve(FloatCurve=DataInterface() {propertyValues {Curve {keys [<at 0 1>,<at 1 0>]}}})
    Scale RGB=&VectorFromFloat(Value=&Multiply_Float(
      A=&FloatFromCurve(FloatCurve=DataInterface() {propertyValues {Curve {keys [<at 0 10>,<at 0.2 3>,<at 1 1>]}}})
      B=1
     ))
    ScaleA=true
    ScaleRGB=true
   )
   /SetVariables(
    Particles.DotCamera=1
    Particles.DotColor:LinearColor=<rgba 0 0.427 0.9216 1>
    Particles.ScaleDots:Vector3f=&VectorFromFloat(Value=&FloatFromCurve(
      FloatCurve=DataInterface() {propertyValues {Curve {keys [
          {arriveTangent=4.1785, interpMode=RCIM_Cubic, leaveTangent=4.1785, tangentMode=RCTM_User, time=0, value=0}
          {arriveTangent=38.1546, interpMode=RCIM_Cubic, leaveTangent=-0.121, tangentMode=RCTM_Break, time=0.1, value=0.8}
          <at 0.3 1.5 RCIM_Cubic>
          {arriveTangent=-1.4, interpMode=RCIM_Cubic, leaveTangent=-1.4, time=0.4, value=0.8}
          {arriveTangent=-0.075, interpMode=RCIM_Cubic, leaveTangent=-12.6314, tangentMode=RCTM_Break, time=0.8, value=0.8}
          {arriveTangent=0.0224, interpMode=RCIM_Cubic, leaveTangent=0.0224, tangentMode=RCTM_User, time=1, value=0}
         ]}}}
      Scale Curve=&Multiply_Float(A=1.8, B=$System.FXScale)
     ))
   ) {isSetParameters=true}
   /SetVariables(
    Particles.OutlineCamera:NiagaraFloat=&Add_Float(
     A=-5
     B=&FloatFromCurve(
      FloatCurve=DataInterface() {propertyValues {Curve {keys [
          {arriveTangent=7.5557, interpMode=RCIM_Cubic, leaveTangent=7.5557, tangentMode=RCTM_User, time=0, value=0}
          {arriveTangent=0.0349, interpMode=RCIM_Cubic, leaveTangent=0.0349, tangentMode=RCTM_User, time=0.1, value=1.15}
          {arriveTangent=-1.3734, interpMode=RCIM_Cubic, leaveTangent=1.4248, tangentMode=RCTM_Break, time=0.2, value=0.8}
          {arriveTangent=0.0401, interpMode=RCIM_Cubic, leaveTangent=0.0401, tangentMode=RCTM_User, time=0.275, value=1.1}
          {interpMode=RCIM_Cubic, leaveTangent=2.5288, tangentMode=RCTM_Break, time=0.35, value=0.9}
          {arriveTangent=0.026, interpMode=RCIM_Cubic, leaveTangent=0.026, tangentMode=RCTM_User, time=0.4, value=1}
          {arriveTangent=-1.3077, interpMode=RCIM_Cubic, leaveTangent=-1.3077, time=0.8, value=0.875}
          <at 1 0 RCIM_Cubic>
         ]}}}
      Scale Curve=-20
     )
    )
    Particles.OutlineColor:LinearColor=<rgba 0 0 0 1>
    Particles.OutlineSize:Vector3f=&Multiply_VectorByFloat(
     Vector=$Particles.Scale
     Float=&FloatFromCurve(
      FloatCurve=DataInterface() {propertyValues {Curve {keys [
          {interpMode=RCIM_Cubic, tangentMode=RCTM_User, time=0, value=1}
          {arriveTangent=-0.3003, interpMode=RCIM_Cubic, leaveTangent=-1, tangentMode=RCTM_Break, time=0.9, value=0.9}
          {interpMode=RCIM_Cubic, leaveTangent=-3.6374, time=1, value=0}
         ]}}}
      Scale Curve=1.4
     )
    )
   ) {isSetParameters=true}
  ]
  rendererClasses [&NiagaraMeshRendererProperties]
  renderers [
   {rendererClass=&NiagaraMeshRendererProperties, rendererIndex=0}
   {rendererClass=&NiagaraMeshRendererProperties, rendererIndex=1}
   {rendererClass=&NiagaraMeshRendererProperties, rendererIndex=2}
  ]
 }
 Orbs_Impact {
  emitterUpdateScript [
   &EmitterState()
   &SpawnBurst_Instantaneous(Spawn Count=2, Spawn Time=0.05, Use Spawn Probability=false, UseLoopCountLimit=false)
   &SpawnBurst_Instantaneous(Spawn Group=1, Use Spawn Probability=false, UseLoopCountLimit=false) {module=SpawnBurst_Instantaneous001}
  ]
  particleSpawnScript [
   /SetVariables(Particles.VisibilityTag=$Emitter.SpawnGroup) {isSetParameters=true}
   /SetVariables(Particles.SizeLerp:NiagaraFloat=&FloatFromCurve(
     FloatCurve=DataInterface() {propertyValues {Curve {keys [
         {interpMode=RCIM_Cubic, tangentMode=RCTM_User, time=0, value=0}
         {arriveTangent=1.4324, interpMode=RCIM_Cubic, leaveTangent=1.4324, tangentMode=RCTM_User, time=0.87, value=0.5596}
         {interpMode=RCIM_Cubic, tangentMode=RCTM_User, time=1, value=1}
        ]}}}
     CurveIndex=&RandomRangeFloat(Minimum=0, Maximum=1)
    )) {isSetParameters=true}
   /SetVariables(Particles.Rand:NiagaraInt32=&RandomRangeInt(Minimum=0, Maximum=1)) {isSetParameters=true}
   &InitializeParticle(
    Color=&LerpLinearColors(
     StartColor=&SelectLinearColorFromArray(
      Array Sampling Mode=::NewEnumerator1
      Direct Array Index=&MakeCustomIntFromBool(Bool=&SetBoolByFloatComparison(A=$Particles.SizeLerp, B=0.1), True INT=1, False INT=0)
      Color Selection Array=DataInterface() {propertyValues {ColorData [<rgba 0 1.2486 2 1>, <rgba 0.7305 0 0.1823 1>]}}
      Random Seed=?
     )
     EndColor=&SelectLinearColorFromArray(
      Array Sampling Mode=::NewEnumerator1
      Direct Array Index=$Particles.Rand
      Color Selection Array=DataInterface() {propertyValues {ColorData [<rgba 0 0.8 1 1>, <rgba 0.9911 0.0921 0.4004 1>]}}
      Random Seed=?
     )
     LerpFactor=&MakeFloatFromInt(Int=$Emitter.SpawnGroup)
    )
    Lifetime=&Lerp_Float(
     A=&RandomRangeFloat(Minimum=0.2, Maximum=0.3)
     B=0.25
     Alpha=&MakeFloatFromInt(Int=$Emitter.SpawnGroup)
    )
    Sprite Rotation Mode=::NewEnumerator2
    Sprite Size Mode=::NewEnumerator3
    Uniform Sprite Size=&Lerp_Float(A=5, B=8, Alpha=$Particles.SizeLerp)
    UsePositionOffset=false
    Write Material Random=true
   ) {module=InitializeParticle001}
   &ShapeLocation(MaskSpawnGroup=false, Override Seed=false, Sphere Radius=5, UseOwnerScale=true) {module=ShapeLocation001}
   &AddVelocity(Velocity=&RandomRangeVector(Minimum=<1 1 1>, Maximum=<-1 -1 -1>), Velocity Speed Scale=300) {module=AddVelocity001}
   /SetVariables(Particles.Intensity:NiagaraFloat=&FloatFromCurve(
     FloatCurve=DataInterface() {propertyValues {Curve {keys [
         {arriveTangent=4.6871, interpMode=RCIM_Cubic, leaveTangent=4.6871, tangentMode=RCTM_User, time=0, value=0}
         {interpMode=RCIM_Cubic, tangentMode=RCTM_User, time=0.7, value=1}
        ]}}}
     CurveIndex=&MakeFloatFromLinearColor(LinearColor=<rgba 1 1 1 1>, Channel:ENiagaraLinearColor_Channels=::NewEnumerator3)
    )) {isSetParameters=true}
   &CameraOffset(Camera Offset Amount=&RandomRangeFloat(Minimum=-5, Maximum=5))
  ]
  particleUpdateScript [
   &ParticleState()
   &Drag(Drag=9, Use Linear Drag=true, Use Rotational Drag=true)
   &SolveForcesAndVelocity(Clamp Velocity=false, Limit Acceleration=false)
   &ScaleSpriteSize(
    Uniform Curve Scale=&Multiply_Float(
     A=$Particles.Intensity
     B=&Waveform(
      Global Phase=$Particles.Age
      "Waveform [1]"=::NewEnumerator12
      "[1] Frequency"=&RandomRangeFloat(Minimum=25, Maximum=50)
      "[1] Amplitude Min/Max"=<0.5 1.5>
     )
    )
    Uniform Curve Sprite Scale=DataInterface() {propertyValues {Curve {keys [
        {arriveTangent=10.8984, interpMode=RCIM_Cubic, leaveTangent=10.8984, tangentMode=RCTM_User, time=0, value=0}
        <at 0.2 1>
        {interpMode=RCIM_Cubic, tangentMode=RCTM_User, time=0.6, value=1}
        {arriveTangent=-5.0292, interpMode=RCIM_Cubic, leaveTangent=-5.0292, tangentMode=RCTM_User, time=1, value=0}
       ]}}}
   )
   &DynamicMaterialParameters(
    Index 0 Param 1=&FloatFromCurve(FloatCurve=DataInterface() {propertyValues {Curve {keys [
         {arriveTangent=0.3767, interpMode=RCIM_Cubic, leaveTangent=0.3767, tangentMode=RCTM_User, time=0.3, value=0}
         {
          arriveTangentWeight=0.2394
          interpMode=RCIM_Cubic
          leaveTangentWeight=0.2394
          tangentMode=RCTM_User
          tangentWeightMode=RCTWM_WeightedBoth
          time=1
          value=1
         }
        ]}}})
    Module.Index0Param0Write=true
    Module.Index0Param1Write=true
    Module.Index0Param2Write=true
    Module.Index0Param3Write=true
   )
   /SetVariables(
    Particles.RingCamera=25
    Particles.RingPosition:NiagaraPosition=$Engine.Emitter.SimulationPosition
    Particles.RingSize:Vector2f=&Multiply_Vector2D(
     A=&Vector2DFromFloat(Value=&RandomRangeFloat(Minimum=25, Maximum=50))
     B=&Vector2DFromFloat(Value=&FloatFromCurve(FloatCurve=DataInterface() {propertyValues {Curve {keys [
           {arriveTangent=2.5608, interpMode=RCIM_Cubic, leaveTangent=2.5608, tangentMode=RCTM_User, time=0, value=0}
           {arriveTangent=0.2812, interpMode=RCIM_Cubic, leaveTangent=0.2812, tangentMode=RCTM_User, time=1, value=1}
          ]}}}))
    )
   ) {isSetParameters=true}
  ]
  rendererClasses [&NiagaraSpriteRendererProperties]
  renderers [
   {rendererClass=&NiagaraSpriteRendererProperties}
   {rendererClass=&NiagaraSpriteRendererProperties, rendererIndex=1}
  ]
 }
 Speedlines_Toon {
  emitterUpdateScript [
   &EmitterState(Inactive Response=::NewEnumerator1, Life Cycle Mode=::NewEnumerator1, Loop Behavior=::NewEnumerator1, Loop Duration=1, UseLoopDelay=true)
   &SpawnBurst_Instantaneous(Use Spawn Probability=false, UseLoopCountLimit=false)
  ]
  particleSpawnScript [
   &InitializeParticle(
    Color=<rgba 1 1 1 8>
    Lifetime=5
    Mass Mode=::NewEnumerator1
    Sprite Rotation Angle=&RandomRangeFloat(Minimum=-360, Maximum=360)
    Sprite Rotation Mode=::NewEnumerator0
    Sprite Size=&Vector2DFromFloat(Value=&Multiply_Float(A=$System.FXScale, B=&RandomRangeFloat(Minimum=88, Maximum=104)))
    Sprite Size Mode=::NewEnumerator4
    UsePositionOffset=false
    Write Material Random=true
   ) {module=InitializeParticle001}
   &FN_Sprite_Attributes(
    "Create Normalized Age "=true
    Create Sprite Rotation Attribute=true
    Create SpriteSize Attribute=true
    Lifetime=0.2
    Max Age=0.15
    SpriteRotation:NiagaraFloat=&RandomRangeFloat(Minimum=-360, Maximum=360)
    SpriteSize=&Multiply_Vector2D(
     A=&Vector2DFromFloat(Value=&RandomRangeFloat(Minimum=115, Maximum=130))
     B=&Vector2DFromFloat(Value=$System.FXScale)
    )
    Write to Module Namespace=true
   )
  ]
  particleUpdateScript [
   &ParticleState()
   &ScaleSpriteSize(Uniform Curve Sprite Scale=DataInterface() {propertyValues {Curve {keys [{arriveTangent=1.3574, interpMode=RCIM_Cubic, leaveTangent=1.3574, tangentMode=RCTM_User, time=0, value=0.6}, <at 1 1.4 RCIM_Cubic>]}}})
   /SetVariables(Particles.SpriteSizeSecond:Vector2f=&Multiply_Vector2D(
     A=$Particles.Initial.FN_Sprite_Attributes.SpriteSize
     B=&Vector2DFromFloat(Value=&FloatFromCurve(
       FloatCurve=DataInterface() {propertyValues {Curve {keys [{arriveTangent=1.3574, interpMode=RCIM_Cubic, leaveTangent=1.3574, tangentMode=RCTM_User, time=0, value=0.8}, <at 1 1.4 RCIM_Cubic>]}}}
       Scale Curve=0.8
      ))
    )) {isSetParameters=true}
   &ScaleColor(Scale Alpha=99, Scale RGB=&VectorFromFloat(Value=0), ScaleA=true, ScaleRGB=true) {module=ScaleColor001}
   &CameraOffset(Camera Offset Amount=25)
   /SetVariables(Particles.New Linear Color:LinearColor=&LerpLinearColors(StartColor=<rgba 2 2 2 5>, EndColor=<rgba 0 0 0 10>, LerpFactor=0)) {isSetParameters=true}
   &DynamicMaterialParameters(
    Index 0 Param 1=1
    Index 0 Param 2=&FloatFromCurve(FloatCurve=DataInterface() {propertyValues {Curve {keys [{interpMode=RCIM_Cubic, tangentMode=RCTM_User, time=0, value=0}, {interpMode=RCIM_Cubic, tangentMode=RCTM_User, time=1, value=0.4}]}}})
    Module.Index0Param0Write=true
    Module.Index0Param1Write=true
    Module.Index0Param2Write=true
    Module.Index0Param3Write=true
   ) {module=DynamicMaterialParameters001}
  ]
  rendererClasses [&NiagaraSpriteRendererProperties]
  renderers [
   {rendererClass=&NiagaraSpriteRendererProperties}
   {rendererClass=&NiagaraSpriteRendererProperties, rendererIndex=1}
  ]
 }
}
summary {
 systemName=NS_Emote_PillowMill
 emitters [
  {emitterName=Orbs_Impact, rendererClasses [&NiagaraSpriteRendererProperties]}
  {emitterName=Speedlines_Toon, rendererClasses [&NiagaraSpriteRendererProperties]}
  {emitterName=Flash_Comic, rendererClasses [&NiagaraMeshRendererProperties]}
 ]
 userVariables []
}
